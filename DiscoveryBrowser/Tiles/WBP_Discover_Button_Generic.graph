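# Discovery browser button widget that extends the UIKit generic button with rounded corners and a focus notification, adding no widgets of its own.
# PreConstruct runs the parent logic, casts the inherited background block to the outlined variant and sets its pixel-based corner radii.
# It then writes a separate, smaller corner radius directly into the Backing image's dynamic material.
# Gaining focus broadcasts the OnDiscoverButtonFocusGained dispatcher.
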WidgetBlueprint /DiscoveryBrowser/Tiles/WBP_Discover_Button_Generic
graphs [
 EventGraph {nodes [
   Event On Focused() {is=title K2Node_Event}
   > Call On Discover Button Focus Gained!(execute:Exec, self:WBP Discover Button Generic Object Reference) -> (then:Exec) {is=title K2Node_CallDelegate}
   n1 = Event PreConstruct()
   > Parent: PreConstruct(IsDesignTime=@n1.IsDesignTime)
   > n2 = Cast To WBP_UIKit_Block_BackgroundWithOutline!(execute, Object=Get Block_Background() -> (Block_Background:WBP UIKit Block Background Base Object Reference)) -> (then, CastFailed, AsWBP UIKit Block Background with Outline:WBP UIKit Block Background With Outline Object Reference) {is=title K2Node_DynamicCast}
   > SetCornerRadiuses!(
    execute
    self:WBP UIKit Block Background With Outline Object Reference=@n2
    TopLeft:float=20
    TopRight:float=20
    BottomLeft:float=20
    BottomRight:float=20
    RadiusInPixels:bool=true
   ) -> (then)
   > n3 = GetDynamicMaterial(self=Get Backing!(self:WBP UIKit Block Background With Outline Object Reference=@n2) -> (Backing:Image Object Reference))
   > SetScalarParameterValue(self=@n3, ParameterName=Corner Radius, Value=10)
   > SetScalarParameterValue(self=@n3, ParameterName=Radius in Pixels, Value=1)
  ]}
 OnDiscoverButtonFocusGained {nodes [OnDiscoverButtonFocusGained() {is=title K2Node_FunctionEntry}]}
]
variables [OnDiscoverButtonFocusGained {}]
widgets {
 info {
  inheritedWidgetCount=0
  parentClass=/FortUILibrary/UIKit/Buttons/WBP_UIKit_Button_Generic.WBP_UIKit_Button_Generic_C
  rootWidgetClass=null
  widgetCount=0
 }
 widgets []
}
