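# Primary-info text block for Discover browser tiles, a CommonUserWidget wrapping a single CommonTextBlock with ellipsis overflow and marquee scroll style.
# Exposes functions to set its text, justification and marquee scrolling, which is off by default.
# SetBlockFocused plays the Focused animation forward or in reverse, fading the text between slightly dimmed and fully opaque white.
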WidgetBlueprint /DiscoveryBrowser/Tiles/WBP_Discover_Tile_TextPrimary
animations [Focused {
  binding_count=1
  length_seconds=0.2
  object=&self:Focused
  sequence {bindings [Text_PrimaryInfo {
     id=#1
     possessed_class=CommonTextBlock
     tracks [{
       type=MovieSceneColorTrack
       display_name=Color and Opacity
       sections [{
         type=MovieSceneColorSection
         end_frame=4
         start_frame=0
         channels [
          Color.R {type=MovieSceneScriptingFloatChannel, const [1,1]}
          Color.G {type=MovieSceneScriptingFloatChannel, const [1,1]}
          Color.B {type=MovieSceneScriptingFloatChannel, const [1,1]}
          Color.A {type=MovieSceneScriptingFloatChannel, keys [[0,0.9],[4,1]]}
         ]
        }]
      }]
    }]}
 }]
graphs [
 EventGraph {nodes [
   Event PreConstruct()
   Event Construct()
   Event Tick()
  ]}
 SetBlockFocused {nodes [
   n1 = SetBlockFocused!() -> (then, IsFocused:bool=~) {is=title K2Node_FunctionEntry}
   > n2 = Branch(Condition=@n1)
   n3 = Get Focused() -> (Focused:Widget Animation Object Reference)
   > PlayAnimationForward(InAnimation=@n3)
   PlayAnimationReverse(execute=@n2.else, InAnimation=@n3)
  ]}
 SetBlockScrollingEnabled {nodes [
   n1 = SetBlockScrollingEnabled!() -> (then, IsMarqueeScrollingEnabled:bool=~) {is=title K2Node_FunctionEntry}
   > SetScrollingEnabled!(
    execute
    self:Common Text Object Reference=Get Text_PrimaryInfo() -> (Text_PrimaryInfo:Common Text Object Reference)
    bInIsScrollingEnabled:bool=@n1
   ) -> (then)
  ]}
 SetBlockText {nodes [
   n1 = SetBlockText!() -> (then, BlockText:Text) {is=title K2Node_FunctionEntry}
   > "SetText (Text)"(self=Get Text_PrimaryInfo(), InText=@n1)
  ]}
 SetBlockTextJustification {nodes [
   n1 = SetBlockTextJustification!() -> (then, TextJustification:ETextJustify Enum=Left) {is=title K2Node_FunctionEntry}
   > Set Justification!(execute, Justification:ETextJustify Enum=@n1, self:Text Layout Widget Object Reference=Get Text_PrimaryInfo()) -> (then, Output_Get:ETextJustify Enum=Left)
  ]}
]
variables []
widgets {
 info {inheritedWidgetCount=0, parentClass=&CommonUserWidget, rootWidgetClass=&CommonTextBlock, widgetCount=1}
 widgets [Text_PrimaryInfo {
   widgetClassPath=&CommonTextBlock
   bIsVariable=true
   widget=Text_PrimaryInfo
   properties {
    bIsScrollingEnabled=false
    mobileFontSizeMultiplier=1.5
    scrollStyle=&TextScrollStyle-Base_C
    text="{VM Primary Info}"
    textOverflowPolicy=Ellipsis
    colorAndOpacity {colorUseRule=UseColor_Specified, specifiedColor=<rgba 1 1 1 0.9>}
    font {
     bForceMonospaced=false
     bMaterialIsStencil=false
     fontMaterial=null
     fontObject=&HeadingNow
     letterSpacing=0
     monospacedWidth=1
     size=14.25
     skewAmount=0
     typefaceFontName=74 Regular
     outlineSettings {
      bApplyOutlineToDropShadows=false
      bMiteredCorners=false
      bSeparateFillAlpha=false
      outlineMaterial=null
      outlineSize=0
      outlineColor=<rgba 0 0 0 1>
     }
    }
    shadowOffset=<0 0>
   }
  }]
}
