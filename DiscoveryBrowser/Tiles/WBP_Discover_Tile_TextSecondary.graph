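# A Common UI text widget that displays the secondary info line on Discover browser tiles, using an outlined HeadingNow font.
# It wraps a single CommonTextBlock and exposes functions to set its text, justification, and color.
# The color function selects either a grey default or a supplied override, and PreConstruct applies the default.
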WidgetBlueprint /DiscoveryBrowser/Tiles/WBP_Discover_Tile_TextSecondary
graphs [
 EventGraph {nodes [
   Event PreConstruct()
   > SetBlockTextColorAndOpacity!(execute, self, UseDefault:bool=true, OverrideColor:Slate Color Structure) -> (then)
   Event Construct()
   Event Tick()
  ]}
 SetBlockText {nodes [
   n1 = SetBlockText!() -> (then, SecondaryText:Text) {is=title K2Node_FunctionEntry}
   > "SetText (Text)"(self=Get Text_TileBlock() -> (Text_TileBlock:Common Text Object Reference), InText=@n1)
  ]}
 SetBlockTextColorAndOpacity {nodes [
   n1 = SetBlockTextColorAndOpacity!() -> (then, UseDefault:bool=~, OverrideColor:Slate Color Structure) {is=title K2Node_FunctionEntry}
   > SetColorAndOpacity(
    self=Get Text_TileBlock()
    InColorAndOpacity=Select(
     Option 0:Slate Color Structure=@n1.OverrideColor
     Option 1:Slate Color Structure=Get DefaultTextColorAndOpacity() -> (DefaultTextColorAndOpacity:Slate Color Structure)
     Index=@n1
    ) -> (ReturnValue:Slate Color Structure)
   )
  ]}
 SetBlockTextJustification {nodes [
   n1 = SetBlockTextJustification!() -> (then, Justification:ETextJustify Enum=Left) {is=title K2Node_FunctionEntry}
   > Set Justification!(execute, Justification:ETextJustify Enum=@n1, self:Text Layout Widget Object Reference=Get Text_TileBlock()) -> (then, Output_Get:ETextJustify Enum=Left)
  ]}
]
variables [DefaultTextColorAndOpacity {type=SlateColor, default {colorUseRule=UseColor_Specified, specifiedColor=<rgba 0.5776 0.5776 0.5776 1>}}]
widgets {
 info {inheritedWidgetCount=0, parentClass=&CommonUserWidget, rootWidgetClass=&CommonTextBlock, widgetCount=1}
 widgets [Text_TileBlock {
   widgetClassPath=&CommonTextBlock
   bIsVariable=true
   widget=Text_TileBlock
   properties {
    mobileFontSizeMultiplier=1.45
    scrollStyle=&TextScrollStyle-Base_C
    text="{Secondary Info}"
    colorAndOpacity {colorUseRule=UseColor_Specified, specifiedColor=<rgba 0.5776 0.5776 0.5776 1>}
    font {
     bForceMonospaced=false
     bMaterialIsStencil=false
     fontMaterial=null
     fontObject=&HeadingNow
     letterSpacing=10
     monospacedWidth=1
     size=14.25
     skewAmount=0
     typefaceFontName=74 Regular
     outlineSettings {
      bApplyOutlineToDropShadows=false
      bMiteredCorners=false
      bSeparateFillAlpha=false
      outlineMaterial=/FortUILibrary/MaterialInstances/Text/MI_UI_Block_Text_Outline
      outlineSize=2
      outlineColor=<rgba 0 0 0 1>
     }
    }
    shadowOffset=<0 0>
   }
  }]
}
