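# Debug overlay widget for Discover browser tiles, showing a tile's analytics timers, interaction state and link code.
# Set View Model stores the debug view model, hands it the owning button and binds that button's selection-changed event to a Selected flag.
# UpdateDebugInfo, gated by a console variable and a valid button, writes visible and selected times, colored by state, tints the border while visible, and lists hovered/focused/selected.
# The link code shows only for activity view models.
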WidgetBlueprint /DiscoveryBrowser/Tiles/WBP_TileDebug
graphs [
 EventGraph {nodes [
   Event PreConstruct()
   Event Construct()
   Event Tick()
  ]}
 On Owning Button Selected Changed {nodes [
   n1 = On Owning Button Selected Changed!() -> (then, Button:Common Button Base Object Reference, Selected:bool=~) {is=title K2Node_FunctionEntry}
   > Set Selected(Selected:bool=@n1.Selected) -> (Output_Get:bool)
  ]}
 Set View Model {nodes [
   n1 = Set View Model!() -> (then, In View Model:Discover Tile Item Debug View Model Object Reference, In Owning Button:Common Button Base Object Reference) {is=title K2Node_FunctionEntry}
   > Set FortDiscoverTileItemDebugVM(FortDiscoverTileItemDebugVM:Discover Tile Item Debug View Model Object Reference=@n1) -> (Output_Get:Discover Tile Item Debug View Model Object Reference)
   > SetOwningButton!(
    execute
    self:Discover Tile Item Debug View Model Object Reference=@n1
    InOwningButton:Common Button Base Object Reference=@n1.In Owning Button
   ) -> (then)
   > Bind Event to On Selected Changed(self:Common Button Base Object Reference=@n1.In Owning Button, Delegate=Create Event()) {is=title K2Node_AddDelegate}
  ]}
 UpdateDebugInfo {nodes [
   UpdateDebugInfo!() -> (then, DummyParam:bool=~) {is=title K2Node_FunctionEntry}
   > n1 = GetConsoleVariableBoolValue!(execute, VariableName:String=Athena.DiscoverSurface.TileItemDebugUI) -> (then, ReturnValue:bool)
   > Branch(Condition=@n1)
   n2 = Get FortDiscoverTileItemDebugVM() -> (FortDiscoverTileItemDebugVM:Discover Tile Item Debug View Model Object Reference)
   > Branch(Condition=IsValid(Object=GetOwningButton!(self:Discover Tile Item Debug View Model Object Reference=@n2) -> (ReturnValue:Common Button Base Object Reference)))
   n3 = Get FortDiscoverTileItemDebugVM()
   n4 = "To String (Text)"(InText=GetTimespanAsText!(
     Timespan:"Timespan Structure (by ref)"=GetTimeVisible!(self:Discover Tile Item Debug View Model Object Reference=@n3) -> (ReturnValue:Timespan Structure)
     bShowHoursWhenZero:bool=true
     bShowSeconds:bool=true
     FormatType:ETimespanAsTextFormat Enum=DaysHoursMinutesSeconds
     UnitDisplayType:ETimespanUnitDisplayFormat Enum=Abbreviated
     bUseSentenceCase:bool=false
    ) -> (ReturnValue:Text))
   n5 = Get Text_Analytics_Visible() -> (Text_Analytics_Visible:Common Text Object Reference)
   > "SetText (Text)"(self=@n5, InText="To Text (String)"(InString=Append(A="Visible: ", B=@n4)))
   n6 = IsVisible(self:Discover Tile Item Debug View Model Object Reference=Get FortDiscoverTileItemDebugVM())
   > SetColorAndOpacity(
    self=@n5
    InColorAndOpacity=Select(
     Option 0:Slate Color Structure=Make Slate Color(SpecifiedColor=(R=0.6,G=0.6,B=0.6,A=1))
     Option 1:Slate Color Structure=Make Slate Color(SpecifiedColor=(R=0,G=1,B=0,A=1))
     Index=@n6
    ) -> (ReturnValue:Slate Color Structure)
   )
   > SetBrushColor(
    self=Get BorderVisible() -> (BorderVisible:Border Object Reference)
    InBrushColor=MakeColor!(R:float=0, G:float=1, B:float=0, A:float=SelectFloat(A=0.6, B=0, bPickA=@n6)) -> (ReturnValue:Linear Color Structure)
   )
   n7 = Get FortDiscoverTileItemDebugVM()
   n8 = "To String (Text)"(InText=GetTimespanAsText(
     Timespan=GetTimeSelected!(self:Discover Tile Item Debug View Model Object Reference=@n7) -> (ReturnValue:Timespan Structure)
     bShowHoursWhenZero=true
     bShowSeconds=true
     FormatType=DaysHoursMinutesSeconds
     UnitDisplayType=Abbreviated
     bUseSentenceCase=false
    ))
   n9 = Get Text_Analytics_Selected() -> (Text_Analytics_Selected:Common Text Object Reference)
   > "SetText (Text)"(self=@n9, InText="To Text (String)"(InString=Append(A="Selected: ", B=@n8)))
   n10 = Get FortDiscoverTileItemDebugVM()
   > SetColorAndOpacity(
    self=@n9
    InColorAndOpacity=Select(
     Option 0=Make Slate Color(SpecifiedColor=(R=0.6,G=0.6,B=0.6,A=1))
     Option 1=Make Slate Color(SpecifiedColor=(R=0,G=1,B=0,A=1))
     Index=IsSelected!(self:Discover Tile Item Debug View Model Object Reference=@n10) -> (ReturnValue:bool)
    )
   )
   > n11 = Sequence!(execute) -> (then_0, then_1, then_2:Exec)
   n12 = Get Text_Widget_State() -> (Text_Widget_State:Common Text Object Reference)
   > SetVisibility(self=@n12, InVisibility=SelfHitTestInvisible)
   n13 = GetOwningButton(self=Get FortDiscoverTileItemDebugVM())
   n14 = Select(Option 0:Text, Option 1:Text="LOCTEXT(\"HOVERED \")", Index=IsHovered(self=@n13)) -> (ReturnValue:Text)
   n15 = "To String (Text)"(InText=@n14)
   n16 = Select(
    Option 0=@n14
    Option 1="To Text (String)"(InString=Append(A=@n15, B="FOCUSED "))
    Index=HasAnyUserFocus!(self:Widget Object Reference=@n13) -> (ReturnValue:bool)
   )
   n17 = Append(A="To String (Text)"(InText=@n16), B="SELECTED  ")
   > "SetText (Text)"(self=@n12, InText=Select(Option 0=@n16, Option 1="To Text (String)"(InString=@n17), Index=Get Selected() -> (Selected:bool)))
   n18 = Cast To FortActivityViewModel!(
    execute=@n11.then_1
    Object=GetTileItemVM!(self:Discover Tile Item Debug View Model Object Reference=Get FortDiscoverTileItemDebugVM()) -> (ReturnValue:Discover Tile Item View Model Object Reference)
   ) -> (then, CastFailed, AsDiscover Activity View Model:Discover Activity View Model Object Reference) {is=title K2Node_DynamicCast}
   n19 = Get Text_LinkCode() -> (Text_LinkCode:Common Text Object Reference)
   > "SetText (Text)"(
    self=@n19
    InText="To Text (String)"(InString=GetMnemonic!(self:Discover Tile Item View Model Object Reference=@n18) -> (ReturnValue:String))
   )
   > SetVisibility(self=@n19, InVisibility=SelfHitTestInvisible)
   SetVisibility(execute=@n18.CastFailed, self=@n19, InVisibility=Collapsed)
  ]}
]
variables [Selected {type=boolean}]
widgets {
 info {inheritedWidgetCount=0, parentClass=&UserWidget, rootWidgetClass=&Overlay, widgetCount=7}
 widgets [
  Overlay_27 {widgetClassPath=&Overlay, widget=Overlay_27}
  BorderVisible {
   widgetClassPath=&Border
   bIsVariable=true
   parent=Overlay_27
   widget=BorderVisible
   slot {class=OverlaySlot, properties {horizontalAlignment=HAlign_Fill, verticalAlignment=VAlign_Fill}}
   properties=~
  }
  VerticalBox {widgetClassPath=&VerticalBox, parent=Overlay_27, widget=VerticalBox, slot {class=OverlaySlot}}
  Text_LinkCode {
   widgetClassPath=&CommonTextBlock
   bIsVariable=true
   parent=VerticalBox
   widget=Text_LinkCode
   properties {
    text=LinkCode : 1234-5678-9100
    colorAndOpacity {colorUseRule=UseColor_Specified, specifiedColor=<rgba 0.6 0.6 0.6 1>}
    font {
     bForceMonospaced=false
     bMaterialIsStencil=false
     fontMaterial=null
     fontObject=&Burbank
     letterSpacing=0
     monospacedWidth=1
     size=15
     skewAmount=0
     typefaceFontName=Burbank-Black
     outlineSettings {
      bApplyOutlineToDropShadows=false
      bMiteredCorners=false
      bSeparateFillAlpha=false
      outlineMaterial=null
      outlineSize=3
      outlineColor=<rgba 0 0 0 1>
     }
    }
    shadowColorAndOpacity=<rgba 1 1 1 1>
    shadowOffset=<0 0>
   }
   slot {class=VerticalBoxSlot, properties {horizontalAlignment=HAlign_Left, verticalAlignment=VAlign_Top}}
  }
  Text_Analytics_Visible {
   widgetClassPath=&CommonTextBlock
   bIsVariable=true
   parent=VerticalBox
   widget=Text_Analytics_Visible
   properties {
    text=Visible : 12 seconds
    colorAndOpacity {colorUseRule=UseColor_Specified, specifiedColor=<rgba 0 1 0 1>}
    font {
     bForceMonospaced=false
     bMaterialIsStencil=false
     fontMaterial=null
     fontObject=&Burbank
     letterSpacing=0
     monospacedWidth=1
     size=15
     skewAmount=0
     typefaceFontName=Burbank-Black
     outlineSettings {
      bApplyOutlineToDropShadows=false
      bMiteredCorners=false
      bSeparateFillAlpha=false
      outlineMaterial=null
      outlineSize=3
      outlineColor=<rgba 0 0 0 1>
     }
    }
    shadowColorAndOpacity=<rgba 1 1 1 1>
    shadowOffset=<0 0>
   }
   slot {class=VerticalBoxSlot, properties {horizontalAlignment=HAlign_Left, verticalAlignment=VAlign_Top}}
  }
  Text_Analytics_Selected {
   widgetClassPath=&CommonTextBlock
   bIsVariable=true
   parent=VerticalBox
   widget=Text_Analytics_Selected
   properties {
    text=Selected : 2 seconds
    colorAndOpacity {colorUseRule=UseColor_Specified, specifiedColor=<rgba 0 1 0 1>}
    font {
     bForceMonospaced=false
     bMaterialIsStencil=false
     fontMaterial=null
     fontObject=&Burbank
     letterSpacing=0
     monospacedWidth=1
     size=15
     skewAmount=0
     typefaceFontName=Burbank-Black
     outlineSettings {
      bApplyOutlineToDropShadows=false
      bMiteredCorners=false
      bSeparateFillAlpha=false
      outlineMaterial=null
      outlineSize=3
      outlineColor=<rgba 0 0 0 1>
     }
    }
    shadowColorAndOpacity=<rgba 1 1 1 1>
    shadowOffset=<0 0>
   }
   slot {class=VerticalBoxSlot, properties {horizontalAlignment=HAlign_Left, verticalAlignment=VAlign_Top}}
  }
  Text_Widget_State {
   widgetClassPath=&CommonTextBlock
   bIsVariable=true
   parent=VerticalBox
   widget=Text_Widget_State
   properties {
    text=Widget State
    colorAndOpacity {colorUseRule=UseColor_Specified, specifiedColor=<rgba 0 1 0 1>}
    font {
     bForceMonospaced=false
     bMaterialIsStencil=false
     fontMaterial=null
     fontObject=&Burbank
     letterSpacing=0
     monospacedWidth=1
     size=15
     skewAmount=0
     typefaceFontName=Burbank-Black
     outlineSettings {
      bApplyOutlineToDropShadows=false
      bMiteredCorners=false
      bSeparateFillAlpha=false
      outlineMaterial=null
      outlineSize=3
      outlineColor=<rgba 0 0 0 1>
     }
    }
    shadowColorAndOpacity=<rgba 1 1 1 1>
    shadowOffset=<0 0>
   }
   slot {class=VerticalBoxSlot, properties {horizontalAlignment=HAlign_Left, verticalAlignment=VAlign_Top}}
  }
 ]
}
