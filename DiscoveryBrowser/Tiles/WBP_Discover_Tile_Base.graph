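# WBP_Discover_Tile_Base is the base widget for Discovery browser tiles, with optional primary text, secondary text and icon blocks built at runtime into grid rows.
# Tile size derives from column count using shared material parameter values, viewport and safe zone, with a mobile multiplier.
# Selection plays a grow animation, focuses the outline and starts view-model impression telemetry.
# A runtime-gated context menu adapts its button to input type and selection.
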WidgetBlueprint /DiscoveryBrowser/Tiles/WBP_Discover_Tile_Base
animations [Anim_OnHoverSelect {
  binding_count=4
  length_seconds=0.16
  object=&self:Anim_OnHoverSelect
  sequence {
   bindings [
    VerticalBox_0 {id=#1, possessed_class=VerticalBox, tracks []}
    SizeBoxSlot_0 {
     id=#2
     parent_id=#3
     possessed_class=SizeBoxSlot
     tracks [{
       type=MovieSceneMarginTrack
       display_name=Padding
       sections [{
         type=MovieSceneMarginSection
         end_frame=9
         start_frame=0
         channels [
          Left {type=MovieSceneScriptingFloatChannel, keys [[0,0],[9,-8]]}
          Top {type=MovieSceneScriptingFloatChannel, keys [[0,0],[9,-8]]}
          Right {type=MovieSceneScriptingFloatChannel, keys [[0,0],[9,-8]]}
          Bottom {type=MovieSceneScriptingFloatChannel, keys [[0,0],[9,-8]]}
         ]
        }]
      }]
    }
    Grid_Root_GrowsOnSelection {id=#3, possessed_class=GridPanel, tracks []}
    SizeBoxSlot_0 {
     id=#4
     parent_id=#1
     possessed_class=SizeBoxSlot
     tracks [{
       type=MovieSceneMarginTrack
       display_name=Padding
       sections [{
         type=MovieSceneMarginSection
         end_frame=9
         start_frame=0
         channels [
          Left {type=MovieSceneScriptingFloatChannel, keys [[0,0],[9.6,8]]}
          Top {type=MovieSceneScriptingFloatChannel, keys [[0,8],[9.6,12]]}
          Right {type=MovieSceneScriptingFloatChannel, keys [[0,0],[9.6,-8]]}
          Bottom {type=MovieSceneScriptingFloatChannel, const [0,1]}
         ]
        }]
      }]
    }
   ]
   display_rate {numerator=60}
  }
 }]
graphs [
 AddIconBlock {nodes [
   AddIconBlock() {is=title K2Node_FunctionEntry}
   > n1 = Branch(Condition=IsValid(Object=Get Block_Icon() -> (Block_Icon:WBP UIKit Block Image Object Reference)))
   n2 = Construct Object from Class(execute=@n1.else, Class=/FortUILibrary/Systems/ModularButton/Blocks/WBP_UIKit_Block_Image.WBP_UIKit_Block_Image_C) -> (ReturnValue:WBP UIKit Block Image Object Reference)
   > n3 = Set Block_Icon(Block_Icon:WBP UIKit Block Image Object Reference=@n2) -> (Output_Get:WBP UIKit Block Image Object Reference)
   AddChildToGrid!(
    execute=[@n1.then, @n3.then]
    self:Grid Panel Object Reference=Get SecondaryRow() -> (SecondaryRow:Grid Panel Object Reference)
    Content:Widget Object Reference=@n3
    InRow:int=0
    InColumn:int=0
   ) -> (then, ReturnValue:Grid Slot Object Reference)
  ]}
 AddPrimaryTextBlock {nodes [
   AddPrimaryTextBlock() {is=title K2Node_FunctionEntry}
   > RemoveWidget!(execute, Widget:Widget Object Reference=Get Block_PrimaryText() -> (Block_PrimaryText:WBP Discover Tile Text Primary Object Reference)) -> (then)
   n1 = Get IsTileNameInside() -> (IsTileNameInside:bool)
   > Branch(Condition=AND Boolean(A=Get UsePrimaryTextBlock() -> (UsePrimaryTextBlock:bool), B=NOT Boolean(A=@n1)))
   > n2 = Construct Object from Class(Class=Get TextClass() -> (TextClass:WBP Discover Tile Text Primary Class Reference)) -> (ReturnValue:WBP Discover Tile Text Primary Object Reference)
   > n3 = Set Block_PrimaryText(Block_PrimaryText:WBP Discover Tile Text Primary Object Reference=@n2) -> (Output_Get:WBP Discover Tile Text Primary Object Reference)
   > Branch(Condition=IsValid(Object=@n3))
   > AddChildToGrid(self=Get PrimaryRow() -> (PrimaryRow:Grid Panel Object Reference), Content=@n3, InRow=0, InColumn=0)
   n4 = SlotAsGridSlot(Widget=Get Block_PrimaryText())
   > SetHorizontalAlignment(self:Grid Slot Object Reference=@n4, InHorizontalAlignment=HAlign_Fill)
   > SetVerticalAlignment(self:Grid Slot Object Reference=@n4, InVerticalAlignment=VAlign_Bottom)
   > UpdatePrimaryTextJustification!(execute, self, Justification:ETextJustify Enum=Get PrimaryTextJustification() -> (PrimaryTextJustification:ETextJustify Enum=Left)) -> (then)
  ]}
 AddSecondaryTextBlock {nodes [
   AddSecondaryTextBlock() {is=title K2Node_FunctionEntry}
   > RemoveWidget(Widget=Get Block_SecondaryText() -> (Block_SecondaryText:WBP Discover Tile Text Secondary Object Reference))
   > Branch(Condition=Get UseSecondaryTextBlock() -> (UseSecondaryTextBlock:bool))
   > n1 = Construct Object from Class(Class=Get SecondaryTextClass() -> (SecondaryTextClass:WBP Discover Tile Text Secondary Class Reference)) -> (ReturnValue:WBP Discover Tile Text Secondary Object Reference)
   > n2 = Set Block_SecondaryText(Block_SecondaryText:WBP Discover Tile Text Secondary Object Reference=@n1) -> (Output_Get:WBP Discover Tile Text Secondary Object Reference)
   > Branch(Condition=IsValid(Object=@n2))
   > AddChildToGrid(self=Get SecondaryRow(), Content=@n2, InRow=0, InColumn=1)
   n3 = SlotAsGridSlot(Widget=Get Block_SecondaryText())
   > SetHorizontalAlignment(self=@n3, InHorizontalAlignment=HAlign_Fill)
   > SetVerticalAlignment(self=@n3, InVerticalAlignment=VAlign_Center)
   > UpdateSecondaryTextJustification!(execute, self, Justification:ETextJustify Enum=Get SecondaryTextJustification() -> (SecondaryTextJustification:ETextJustify Enum=Left)) -> (then)
   > UpdateSecondaryTextColorAndOpacity!(execute, self, UseDefault:bool=true, OverrideColor:Slate Color Structure) -> (then)
  ]}
 CleanupContextMenuOnRelease {nodes [
   CleanupContextMenuOnRelease() {is=title K2Node_FunctionEntry}
   > Branch(Condition=Get bEnableContextMenuInDiscover!(self:Fort Runtime Options Object Reference=GetRuntimeOptions()) -> (bEnableContextMenuInDiscover:bool))
   n1 = Get Anchor!(self:UIKit Context Menu Anchor Object Reference=Get WBP_UIKit_ContextMenuAnchorButton() -> (WBP_UIKit_ContextMenuAnchorButton:WBP UIKit Context Menu Anchor Button Object Reference)) -> (Anchor:Menu Anchor Object Reference)
   > n2 = Branch(Condition=IsOpen!(self:Menu Anchor Object Reference=@n1) -> (ReturnValue:bool))
   > n3 = Close(self:Menu Anchor Object Reference=@n1)
   n4 = Get WBP_BottomBarHackButton_ContextMenu() -> (WBP_BottomBarHackButton_ContextMenu:WBP Bottom Bar Decoy Button Object Reference)
   Branch(
    execute=[@n3.then, @n2.else]
    Condition="Not Equal (Boolean)"(A=GetInputAction!(self:Common Button Base Object Reference=@n4) -> (InputActionRow:Data Table Row Handle Structure, ReturnValue:bool).ReturnValue) -> (ReturnValue=~)
   )
   > SetBottomButtonVisibility!(execute, self:WBP Bottom Bar Decoy Button Object Reference=@n4, ShouldButtonBeVisible:bool=false) -> (then)
  ]}
 ClearDebugUI {nodes [
   ClearDebugUI() {is=title K2Node_FunctionEntry}
   > Set View Model!(
    execute
    self:WBP Tile Debug Object Reference=Get TileDebug() -> (TileDebug:WBP Tile Debug Object Reference)
    In View Model:Discover Tile Item Debug View Model Object Reference
    In Owning Button:Common Button Base Object Reference
   ) -> (then) {is=title K2Node_CallFunction}
  ]}
 ColumnSizeToTileSize {nodes [
   ColumnSizeToTileSize!() -> (then, NewColumnSize:int, IsSquareTiles:bool=~) {is=title K2Node_FunctionEntry}
   > n1 = GetScalarParameterValue!(
    execute
    Collection:Material Parameter Collection Object Reference=/DiscoveryBrowser/DataAssets/DiscoverySharedValues
    ParameterName:Name=LayoutGridSingleColumnWidth
   ) -> (then, ReturnValue:float) {is=title K2Node_CallMaterialParameterCollectionFunction}
   > Set GridSingleColumnWidth(GridSingleColumnWidth:double=@n1) -> (Output_Get:double)
   > n2 = GetScalarParameterValue(Collection=&DiscoverySharedValues, ParameterName=LayoutGridSingleGutterWidth)
   > Set GridSingleGutterWidth(GridSingleGutterWidth:double=@n2) -> (Output_Get:double)
   > n3 = GetScalarParameterValue(Collection=&DiscoverySharedValues, ParameterName=16x9Ratio)
   > Set Ratio16x9ConvertWidthToHeight(Ratio16x9ConvertWidthToHeight:double=@n3) -> (Output_Get:double)
   > n4 = Branch(Condition=UseNumberOfColumns!() -> (ReturnValue:bool))
   > n5 = GetSafeZonePadding!(execute) -> (
    then
    SafePadding_X:double
    SafePadding_Y:double
    SafePadding_Z:double
    SafePadding_W:double
    SafePaddingScale:Vector 2D Structure
    SpillOverPadding:Vector 4 Structure
   )
   > n6 = GetConsoleVariableBoolValue!(execute, VariableName:String=Frontend.IncludeSafeZoneIntoTileMath) -> (then, ReturnValue:bool)
   n7 = int - int(A=Get NewColumnSize() -> (NewColumnSize:int=0))
   n8 = Get GridSingleColumnWidth() -> (GridSingleColumnWidth:double)
   n9 = GetGeneralRowPadding!() -> (ReturnValue:double)
   n10 = float - float(
    A=SafeDivide(A=GetViewportSize!() -> (ReturnValue_X:double, ReturnValue_Y:double), B=GetViewportScale!() -> (ReturnValue:float))
    B=SelectFloat(A=float + float(A=@n5, B=@n5.SafePadding_Z), B=0, bPickA=@n6)
   )
   n11 = float / float(
    A=float - float(A=float - float(A=@n10, B=@n9), B=float * float(A:int=@n7, B=@n8) {is=title K2Node_PromotableOperator})
    B:int=Get NewColumnSize() -> (NewColumnSize=0)
   )
   n12 = Get IsJoinablePartyTile() -> (IsJoinablePartyTile:bool)
   n13 = Select(
    Option 0:double=Get Ratio16x9ConvertWidthToHeight() -> (Ratio16x9ConvertWidthToHeight:double)
    Option 1:double=1
    Index=AND Boolean(A=Get IsSquareTiles() -> (IsSquareTiles:bool), B=NOT Boolean(A=@n12))
   ) -> (ReturnValue:double)
   > Return Node!(execute, NewTileSize:Vector 2D Structure=MakeVector2D(X=@n11, Y=float * float(A:double=@n11, B=@n13)))
   n14 = Get NewColumnSize() -> (NewColumnSize=0)
   n15 = float + float(
    A=float * float(A:int=@n14, B=Get GridSingleColumnWidth())
    B=float * float(A=int - int(A=@n14), B=Get GridSingleGutterWidth() -> (GridSingleGutterWidth:double))
   )
   n16 = AND Boolean(A=Get IsSquareTiles(), B=NOT Boolean(A=Get IsJoinablePartyTile()))
   n17 = Get Ratio16x9ConvertWidthToHeight()
   n18 = Ceil(A=float * float(A:double=@n15, B=Select(Option 0=@n17, Option 1=1, Index=@n16))) {is=title K2Node_CallFunction}
   Return Node(execute=@n4.else, NewTileSize=MakeVector2D(X=@n15, Y="To Float (Integer)"(InInt=@n18)))
  ]}
 ConfigureDebugUI {nodes [
   n1 = ConfigureDebugUI!() -> (then, Debug VM:Discover Tile Item Debug View Model Object Reference, OwningButton:Common Button Base Object Reference) {is=title K2Node_FunctionEntry}
   > n2 = GetConsoleVariableBoolValue(VariableName=Athena.DiscoverSurface.TileItemDebugUI)
   > Branch(Condition=@n2)
   > n3 = Sequence()
   > n4 = Branch(Condition=IsValid(Object=Get DebugGridSlot() -> (DebugGridSlot:Grid Slot Object Reference)))
   n5 = Create Widget!(
    execute:Exec=@n4.else
    Class:User Widget Class Reference=&/WBP_TileDebug.WBP_TileDebug_C
    OwningPlayer:Player Controller Object Reference=GetOwningPlayer()
    FortDiscoverTileItemDebugVM:Discover Tile Item Debug View Model Object Reference
   ) -> (then:Exec, ReturnValue:WBP Tile Debug Object Reference)
   > Branch(Condition=IsValid(Object=@n5))
   > n6 = Set TileDebug(TileDebug:WBP Tile Debug Object Reference=@n5) -> (Output_Get:WBP Tile Debug Object Reference)
   > n7 = AddChildToGrid(self=Get Grid_Main() -> (Grid_Main:Grid Panel Object Reference), Content=@n6, InRow=0, InColumn=0)
   > n8 = Set DebugGridSlot(DebugGridSlot:Grid Slot Object Reference=@n7) -> (Output_Get:Grid Slot Object Reference)
   > SetLayer!(execute, self:Grid Slot Object Reference=@n8, InLayer:int=1) -> (then)
   Set View Model(execute=@n3.then_1, self=Get TileDebug(), In View Model=@n1, In Owning Button=@n1.OwningButton)
  ]}
 DesignTimeSetTestColumnSize {nodes [
   n1 = DesignTimeSetTestColumnSize!() -> (then, NewSize:int, isSquareTiles:bool=~) {is=title K2Node_FunctionEntry}
   > Set Tile Size!(
    execute
    self
    InSize:Vector 2D Structure=ColumnSizeToTileSize!(NewColumnSize:int=@n1, IsSquareTiles:bool=@n1.isSquareTiles) -> (NewTileSize:Vector 2D Structure) {is=title K2Node_CallFunction}
   ) -> (then) {is=title K2Node_CallFunction}
  ]}
 EventGraph {nodes [
   "On Clicked (WBP_BottomBarHackButton_ContextMenu)"!() -> (OutputDelegate, then, Button:Common Button Base Object Reference) {is=title K2Node_ComponentBoundEvent}
   > Branch(Condition=Get bEnableContextMenuInDiscover(self=GetRuntimeOptions()))
   > ToggleOpen!(execute, self:Menu Anchor Object Reference=Get Anchor(self=Get WBP_UIKit_ContextMenuAnchorButton()), bFocusOnOpen:bool=false) -> (then)
   > Call On Context Menu Toggle!(execute:Exec, self:WBP Discover Tile Base Object Reference, DiscoverTile:WBP Discover Tile Base Object Reference=Self-Reference()) -> (then:Exec) {is=title K2Node_CallDelegate}
   CloseContextMenu() {is=title K2Node_CustomEvent}
   > Close(self=Get Anchor(self=Get WBP_UIKit_ContextMenuAnchorButton()))
   n1 = RemoveInteractionByName!() -> (OutputDelegate, then, ActionToRemove:Name) {is=title K2Node_CustomEvent}
   n2 = Get ContextMenuActionContext() -> (ContextMenuActionContext:UIKit Context Menu Action Context Object Reference)
   > RemoveInteraction!(execute, self:UIKit Context Menu Action Context Object Reference=@n2, InteractionNameID:Name=@n1.ActionToRemove) -> (then)
   RemoveAllInteractions() {is=title K2Node_CustomEvent}
   > RemoveAllInteractions(self:UIKit Context Menu Action Context Object Reference=@n2) {is=title K2Node_CallFunction}
   SetupContextMenu() {is=title K2Node_CustomEvent}
   > Branch(Condition=Get bEnableContextMenuInDiscover(self=GetRuntimeOptions()))
   > n3 = Construct Object from Class(Class=/Script/FortniteUI.UIKitContextMenuActionContext) -> (ReturnValue:UIKit Context Menu Action Context Object Reference)
   > Set ContextMenuActionContext(ContextMenuActionContext:UIKit Context Menu Action Context Object Reference=@n3) -> (Output_Get:UIKit Context Menu Action Context Object Reference)
   > SetContext!(
    execute
    self:WBP UIKit Context Menu Anchor Button Base Object Reference=Get WBP_UIKit_ContextMenuAnchorButton()
    InteractionContext_Data:UIKit Context Menu Action Context Object Reference=Get ContextMenuActionContext()
   ) -> (then)
   HandleInputMethodChanged!() -> (OutputDelegate, then, NewInputType:ECommonInputType Enum) {is=title K2Node_CustomEvent}
   > UpdateContextMenuButtonsVisibility()
   n4 = Event PreConstruct()
   > n5 = Sequence!(execute) -> (then_0, then_1, then_2:Exec)
   > Branch(Condition=@n4.IsDesignTime)
   n6 = Get IsJoinablePartyTile()
   > Branch(Condition=AND Boolean!(A=NOT Boolean(A=@n6), B=Get DesignTimeIsSquareTiles() -> (DesignTimeIsSquareTiles:bool), C:bool=@n4.IsDesignTime) -> (ReturnValue))
   > DesignTimeSetTestColumnSize!(execute, self, NewSize:int=5, isSquareTiles:bool=Get DesignTimeIsSquareTiles()) -> (then) {is=title K2Node_CallFunction}
   UpdatePrimaryTextBlock!(execute=@n5.then_1, self, UsePrimaryText:bool=Get UsePrimaryTextBlock()) -> (then)
   > n7 = Branch(Condition=Get IsTileNameInside())
   UpdateSecondaryTextBlock!(execute=@n7.else, self, UseSecondaryText:bool=Get UseSecondaryTextBlock()) -> (then)
   > n8 = Branch(Condition=Get UseIconBlock() -> (UseIconBlock:bool))
   > AddIconBlock() {is=title K2Node_CallFunction}
   > UpdateIcon()
   RemoveIconBlock(execute=@n8.else)
   SetOutlineSquareOrCircular!(execute=@n5.then_2, self, IsOutlineCircular:bool=Get IsOutlineBlockCircular() -> (IsOutlineBlockCircular:bool)) -> (then)
   > Set Tile Size(InSize=Get TileSize() -> (TileSize:Vector 2D Structure))
   Event Construct()
   n9 = Create Event()
   n10 = Get CommonInputSubsystem()
   > Bind Event to On Input Method Changed(self=@n10, Delegate=@n9)
   > SetupContextMenu() {is=title K2Node_CallFunction}
   > UpdateContextMenuButtonsVisibility()
   Event Destruct()
   > Unbind Event from On Input Method Changed(self=@n10, Delegate=@n9)
   > Unbind all Events from On Context Menu Visibility Changed!(execute:Exec, self:WBP Discover Tile Base Object Reference) -> (then:Exec) {is=title K2Node_ClearDelegate}
  ]}
 Get Dynamic Tile Size {nodes [
   Get Dynamic Tile Size() {is=title K2Node_FunctionEntry}
   > Return Node!(execute, DynamicTileSize:Vector 2D Structure=Get DynamicCalculatedTileSize() -> (DynamicCalculatedTileSize:Vector 2D Structure))
  ]}
 GetGeneralRowPadding {nodes [
   GetGeneralRowPadding() {is=title K2Node_FunctionEntry}
   > n1 = GetConsoleVariableFloatValue!(execute, VariableName:String=Fort.Frontend.RowXZNewPadding) -> (then, ReturnValue:float)
   > Return Node!(execute, ReturnValue:double=float + float(A=@n1, B=236))
  ]}
 OnContextMenuToggle {nodes [OnContextMenuToggle!() -> (then, DiscoverTile:WBP Discover Tile Base Object Reference) {is=title K2Node_FunctionEntry}]}
 OnContextMenuVisibilityChanged {nodes [OnContextMenuVisibilityChanged!() -> (then, IsVisible:bool=~) {is=title K2Node_FunctionEntry}]}
 OnRightClickMouse {nodes [
   n1 = OnRightClickMouse!() -> (then, MouseEvent:Pointer Event Structure) {is=title K2Node_FunctionEntry}
   > Branch(Condition=Get bEnableContextMenuInDiscover(self=GetRuntimeOptions()))
   > n2 = Branch(Condition=Is Mouse Button Down!(Input:"Pointer Event Structure (by ref)"=@n1, MouseButton:Key Structure=RightMouseButton) -> (ReturnValue:bool))
   > Open!(execute, self:Menu Anchor Object Reference=Get Anchor(self=Get WBP_UIKit_ContextMenuAnchorButton()), bFocusMenu:bool=false) -> (then)
   > Play Right Click Sound()
   > Return Node!(execute, WasHandled:Event Reply Structure=Handled())
   Return Node(execute=@n2.else, WasHandled=Unhandled())
  ]}
 OnSetPrimaryTextFromVM {nodes [OnSetPrimaryTextFromVM() {is=title K2Node_FunctionEntry}]}
 OnSetTileContainerSizeOverride {nodes [OnSetTileContainerSizeOverride() {is=title K2Node_FunctionEntry}]}
 OnSetTileIndex {nodes [OnSetTileIndex() {is=title K2Node_FunctionEntry}]}
 OnTileBaseSelectionActive {nodes [
   OnTileBaseSelectionActive() {is=title K2Node_FunctionEntry}
   > Set IsTileActive(IsTileActive:bool=true) -> (Output_Get:bool)
   > RequestTransition!(
    execute
    self:UIKit Block Object Reference=Get Block_Outline() -> (Block_Outline:WBP UIKit Block Outline Object Reference)
    State:EUIKitBlockVisualState Enum=Focused
   ) -> (then)
   > StartTileActiveImpressions()
   > PlayAnimationForward(InAnimation=Get Anim_OnHoverSelect() -> (Anim_OnHoverSelect:Widget Animation Object Reference))
   > UpdateContextMenuButtonsVisibility()
   > Set Visibility(
    Visibility=HitTestInvisible
    self=Get WBP_CaptureForPostBufferUpdate() -> (WBP_CaptureForPostBufferUpdate:WBP Capture For Post Buffer Update Object Reference)
   )
  ]}
 OnTileBaseSelectionNotActive {nodes [
   OnTileBaseSelectionNotActive() {is=title K2Node_FunctionEntry}
   > Set IsTileActive(IsTileActive=false)
   > RequestTransition(self=Get Block_Outline(), State=Unfocused)
   > StopTileActiveImpressions()
   > PlayAnimationReverse(InAnimation=Get Anim_OnHoverSelect(), PlaybackSpeed=4)
   > UpdateContextMenuButtonsVisibility()
   > Set Visibility(Visibility=Collapsed, self=Get WBP_CaptureForPostBufferUpdate())
  ]}
 Override Right Click Sound {nodes [
   n1 = Override Right Click Sound!() -> (then, New Right Click Sound:Sound Base Object Reference) {is=title K2Node_FunctionEntry}
   > Is Valid(InputObject=@n1)
   > Set Right Clicked Sound(Right Clicked Sound:Sound Base Object Reference=@n1) -> (Output_Get:Sound Base Object Reference)
  ]}
 Play Right Click Sound {nodes [
   Play Right Click Sound() {is=title K2Node_FunctionEntry}
   > PlaySound2D(Sound=Get Right Clicked Sound() -> (Right Clicked Sound:Sound Base Object Reference))
  ]}
 RemoveIconBlock {nodes [
   RemoveIconBlock() {is=title K2Node_FunctionEntry}
   n1 = Get Block_Icon()
   > Branch(Condition=IsValid(Object=@n1))
   > RemoveFromParent(self=@n1)
   > Set Block_Icon()
  ]}
 RemovePrimaryTextBlock {nodes [
   RemovePrimaryTextBlock() {is=title K2Node_FunctionEntry}
   n1 = Get Block_PrimaryText()
   > Branch(Condition=IsValid(Object=@n1))
   > RemoveFromParent(self=@n1)
   > Set Block_PrimaryText()
  ]}
 RemoveSecondaryTextBlock {nodes [
   RemoveSecondaryTextBlock() {is=title K2Node_FunctionEntry}
   n1 = Get Block_SecondaryText()
   > Branch(Condition=IsValid(Object=@n1))
   > RemoveFromParent(self=@n1)
   > Set Block_SecondaryText()
  ]}
 ResetAnimationState {nodes [
   ResetAnimationState() {is=title K2Node_FunctionEntry}
   > StopAnimation(InAnimation=Get Anim_Hover!(self:WBP UIKit Block Outline Object Reference=Get Block_Outline()) -> (Anim_Hover:Widget Animation Object Reference))
   > StopAnimation(InAnimation=Get Anim_Sheen!(self:WBP UIKit Block Outline Object Reference=Get Block_Outline()) -> (Anim_Sheen:Widget Animation Object Reference))
   n1 = Get Block_Outline()
   n2 = Get Anim_Unhover!(self:WBP UIKit Block Outline Object Reference=@n1) -> (Anim_Unhover:Widget Animation Object Reference)
   > PlayAnimation(self=@n1, InAnimation=@n2, StartAtTime=GetEndTime(self=@n2))
   n3 = Get Anim_OnHoverSelect()
   n4 = Get IsTileActive() -> (IsTileActive:bool)
   > PlayAnimation(
    InAnimation=@n3
    StartAtTime=Select(
     Option 0:float=GetEndTime(self=@n3)
     Option 1:float=GetStartTime!(self:Widget Animation Object Reference=@n3) -> (ReturnValue:float)
     Index=@n4
    ) -> (ReturnValue:float)
    PlayMode=Select(Option 0:EUMGSequencePlayMode Enum=Reverse, Option 1:EUMGSequencePlayMode Enum, Index=@n4) -> (ReturnValue:EUMGSequencePlayMode Enum)
   )
   > StopAnimation(InAnimation=Get Anim_OnHoverSelect())
  ]}
 SendFlowRiderClickTileEvent {nodes [
   n1 = SendFlowRiderClickTileEvent!() -> (then, TileWidget:User Widget Object Reference, TileType:EActivityItemType Enum=Island, TileUniqueIdentifier:String) {is=title K2Node_FunctionEntry}
   n2 = Make Flow Analytics Additional Info!() -> (FlowAnalyticsAdditionalInfo:Flow Analytics Additional Info Structure) {is=title K2Node_MakeStruct}
   n3 = Get ParentPanelName() -> (ParentPanelName:String)
   n4 = Enum to String!(Enumerator:EActivityItemType Enum=@n1.TileType) -> (ReturnValue:String)
   > Fire Event Element Interacted(
    Element=@n1
    AdditionalInfo="Add String (Non-Localized)"!(
     Info:"Flow Analytics Additional Info Structure (by ref)"="Add String (Non-Localized)"(Info=@n2, Key=PanelName, Value=@n3)
     Key:String=Append(A=@n4, B=" ID")
     Value:String=@n1.TileUniqueIdentifier
    ) -> (ReturnValue:Flow Analytics Additional Info Structure)
   )
  ]}
 Set Tile Container Size Override {nodes [
   n1 = Set Tile Container Size Override!() -> (then, TileSizeOverrideWithOutsideContent:Vector 2D Structure) {is=title K2Node_FunctionEntry}
   > Set DynamicCalculatedTileSize(DynamicCalculatedTileSize:Vector 2D Structure=@n1) -> (Output_Get:Vector 2D Structure)
  ]}
 Set Tile Size {nodes [
   n1 = Set Tile Size!() -> (then, InSize:Vector 2D Structure) {is=title K2Node_FunctionEntry}
   > Set TileSize(TileSize:Vector 2D Structure=@n1) -> (Output_Get:Vector 2D Structure)
   n2 = GetMobileDiscoverTileSizeMultiplier!(self:Fort Runtime Options Object Reference=GetRuntimeOptions()) -> (ReturnValue:float)
   n3 = Select!(
    Option 0:Vector 2D Structure=@n1
    Option 1:Vector 2D Structure=vector2d * vector2d(A:Vector 2D Structure=@n1, B:Vector 2D Structure=MakeVector2D(X=@n2, Y=@n2)) -> (ReturnValue:Vector 2D Structure) {is=title K2Node_PromotableOperator}
    Index=IsMobileGame()
   ) -> (ReturnValue_X:double, ReturnValue_Y:double)
   > SetWidthOverride(
    self=[
     Get SizeBox_Main() -> (SizeBox_Main:Size Box Object Reference)
     Get VisualAttachment_OutsideContent() -> (VisualAttachment_OutsideContent:Fort Visual Attachment Object Reference)
    ]
    InWidthOverride=@n3
   )
   > SetHeightOverride(self=Get SizeBox_Main(), InHeightOverride=@n3.ReturnValue_Y)
   > Set DynamicCalculatedTileSize(DynamicCalculatedTileSize=MakeVector2D(X=@n3, Y=@n3.ReturnValue_Y))
   > Call On Set Tile Container Size Override!(execute:Exec, self:WBP Discover Tile Base Object Reference) -> (then:Exec) {is=title K2Node_CallDelegate}
  ]}
 SetContextMenuIconAlwaysVisible {nodes [
   n1 = SetContextMenuIconAlwaysVisible!() -> (then, IsContextMenuIconAlwaysVisible:bool=~) {is=title K2Node_FunctionEntry}
   > Set IsContextMenuIconAlwaysVisible(IsContextMenuIconAlwaysVisible:bool=@n1) -> (Output_Get:bool)
   > UpdateContextMenuButtonsVisibility()
  ]}
 SetIsPagination {nodes [
   n1 = SetIsPagination!() -> (then, bInIsPagination:bool=~) {is=title K2Node_FunctionEntry}
   n2 = Get TileItemViewModel() -> (TileItemViewModel:Discover Tile Item View Model Object Reference)
   > Is Valid(InputObject=@n2)
   > SetIsPagination!(execute, self:Discover Tile Item View Model Object Reference=@n2, bInIsPagination:bool=@n1) -> (then) {is=title K2Node_CallFunction}
  ]}
 SetOutlineSquareOrCircular {nodes [
   n1 = SetOutlineSquareOrCircular!() -> (then, IsOutlineCircular:bool=~) {is=title K2Node_FunctionEntry}
   > n2 = Set IsOutlineBlockCircular(IsOutlineBlockCircular:bool=@n1) -> (Output_Get:bool)
   n3 = SelectFloat(A=0.5, B=20, bPickA=@n2)
   > SetCornerRadiuses!(
    execute
    self:WBP UIKit Block Outline Object Reference=Get Block_Outline()
    TopLeft:float=@n3
    TopRight:float=@n3
    BottomLeft:float=@n3
    BottomRight:float=@n3
    RadiusInPixels:bool=NOT Boolean(A=@n2)
   ) -> (then)
  ]}
 SetParentOfRightSlot {nodes [
   n1 = SetParentOfRightSlot!() -> (then, IsPrimaryRow:bool=~) {is=title K2Node_FunctionEntry}
   > Branch(Condition=IsValid(Object=Get Block_Icon()))
   > n2 = Branch(Condition=@n1)
   > AddChildToGrid(self=Get PrimaryRow(), Content=Get Slot_Right() -> (Slot_Right:Named Slot Object Reference), InRow=0, InColumn=1)
   AddChildToGrid(execute=@n2.else, self=Get SecondaryRow(), Content=Get Slot_Right(), InRow=0, InColumn=2)
  ]}
 SetParentPanelName {nodes [
   n1 = SetParentPanelName!() -> (then, PanelName:String) {is=title K2Node_FunctionEntry}
   > Set ParentPanelName(ParentPanelName:String=@n1) -> (Output_Get:String)
   n2 = Get WBP_UIKit_ContextMenuAnchorButton()
   > AddInfoForContextMenuTelemetry!(
    execute
    self:WBP UIKit Context Menu Anchor Button Base Object Reference=@n2
    InfoKey:String=PanelName
    InfoValue:String=Get ParentPanelName()
   ) -> (then)
   > SetParentAdditionalDetailsForContextMenuTelemetry!(
    execute
    self:WBP UIKit Context Menu Anchor Button Base Object Reference=@n2
    OptionalAdditionalDetails:Flow Analytics Additional Info Structure="Add String (Non-Localized)"(Info=Make Flow Analytics Additional Info(), Key=PanelName, Value=Get ParentPanelName())
   ) -> (then)
  ]}
 SetPrimaryText {nodes [
   n1 = SetPrimaryText!() -> (then, PrimaryText:Text) {is=title K2Node_FunctionEntry}
   n2 = Get Block_PrimaryText()
   > Branch(Condition=IsValid(Object=@n2))
   > SetBlockText!(execute, self:WBP Discover Tile Text Primary Object Reference=@n2, BlockText:Text=@n1) -> (then)
  ]}
 SetPrimaryTextFocused {nodes [
   n1 = SetPrimaryTextFocused!() -> (then, IsFocused:bool=~) {is=title K2Node_FunctionEntry}
   n2 = Get Block_PrimaryText()
   > Branch(Condition=IsValid(Object=@n2))
   > SetBlockFocused!(execute, self:WBP Discover Tile Text Primary Object Reference=@n2, IsFocused:bool=@n1) -> (then)
  ]}
 SetSecondaryText {nodes [
   n1 = SetSecondaryText!() -> (then, SecondaryText:Text) {is=title K2Node_FunctionEntry}
   n2 = Get Block_SecondaryText()
   > Branch(Condition=IsValid(Object=@n2))
   > SetBlockText!(execute, self:WBP Discover Tile Text Secondary Object Reference=@n2, SecondaryText:Text=@n1) -> (then)
  ]}
 SetTileIndex {nodes [
   n1 = SetTileIndex!() -> (then, TileIndex:int) {is=title K2Node_FunctionEntry}
   > Set TileItemIndex(TileItemIndex:int=@n1) -> (Output_Get:int=0)
   > Call On Set Tile Index!(execute:Exec, self:WBP Discover Tile Base Object Reference) -> (then:Exec) {is=title K2Node_CallDelegate}
  ]}
 SetTileItemViewModel {nodes [
   n1 = SetTileItemViewModel!() -> (then, TileItemViewModel:Discover Tile Item View Model Object Reference) {is=title K2Node_FunctionEntry}
   > Set TileItemViewModel(TileItemViewModel:Discover Tile Item View Model Object Reference=@n1) -> (Output_Get:Discover Tile Item View Model Object Reference)
  ]}
 StartTileActiveImpressions {nodes [
   StartTileActiveImpressions() {is=title K2Node_FunctionEntry}
   > n1 = Branch(Condition=IsValid(Object=Get TileItemViewModel()))
   > StartActiveImpression(self:Discover Tile Item View Model Object Reference=Get TileItemViewModel())
   PrintString!(
    execute=@n1.else
    InString="Warning: Tile Item VM is empty.  ACTIVE IMPRESSIONS FOR TELEMETRY WILL NOT BE SENT FOR THIS TILE TYPE"
    bPrintToScreen
    bPrintToLog
    TextColor
    Duration
    Key
   ) -> (then)
  ]}
 StartTileVisibilityImpression {nodes [
   StartTileVisibilityImpression() {is=title K2Node_FunctionEntry}
   > Is Valid(InputObject=Get TileItemViewModel())
   > StartVisibilityImpression(self:Discover Tile Item View Model Object Reference=Get TileItemViewModel())
  ]}
 StopTileActiveImpressions {nodes [
   StopTileActiveImpressions() {is=title K2Node_FunctionEntry}
   > n1 = Branch(Condition=IsValid(Object=Get TileItemViewModel()))
   > StopActiveImpression(self:Discover Tile Item View Model Object Reference=Get TileItemViewModel())
   PrintString(execute=@n1.else, InString="Warning: Tile Item VM is empty.  ACTIVE IMPRESSIONS FOR TELEMETRY WILL NOT BE STOPPED FOR THIS TILE TYPE")
  ]}
 StopTileVisibilityImpression {nodes [
   StopTileVisibilityImpression() {is=title K2Node_FunctionEntry}
   > Is Valid(InputObject=Get TileItemViewModel())
   > StopVisibilityImpression(self:Discover Tile Item View Model Object Reference=Get TileItemViewModel())
  ]}
 Update Secondary Text as CCU {nodes [
   n1 = Update Secondary Text as CCU!() -> (then, CCUCount:int) {is=title K2Node_FunctionEntry}
   > n2 = Branch(Condition="integer > integer"(A=@n1, B=999))
   > n3 = SetSecondaryText!(
    execute
    self
    SecondaryText:Text=GetShorthandQuantity!(Quantity:int=@n1, MinimumFractionalDigits:int=0, MaximumFractionalDigits:int=1, bShouldRoundDown:bool=false) -> (ReturnValue:Text)
   ) -> (then) {is=title K2Node_CallFunction}
   n4 = "integer <= integer"(A=@n1)
   SetParentOfRightSlot!(execute=[@n6.then, @n3.then], self, IsPrimaryRow:bool=@n4) -> (then) {is=title K2Node_CallFunction}
   n5 = NOT Boolean(A=@n4)
   > UpdateSecondaryTextVisibility!(execute, self, IsVisible:bool=@n5) -> (then)
   > UpdateIconVisibility!(execute, self, IsVisible:bool=@n5) -> (then)
   n6 = SetSecondaryText(execute=@n2.else, SecondaryText="To Text (Integer)"(Value=@n1))
  ]}
 Update Secondary Text as Like Percent {nodes [
   n1 = Update Secondary Text as Like Percent!() -> (then, LikePercent:int) {is=title K2Node_FunctionEntry}
   > SetSecondaryText(SecondaryText=Format Text!(Format="LOCTEXT(\"{p}%\")", p:Text="To Text (Integer)"(Value=@n1)) -> (Result))
   n2 = "integer <= integer"(A=@n1)
   > SetParentOfRightSlot(IsPrimaryRow=@n2)
   n3 = NOT Boolean(A=@n2)
   > UpdateSecondaryTextVisibility(IsVisible=@n3)
   > UpdateIconVisibility(IsVisible=@n3)
  ]}
 UpdateContextMenuButtonsVisibility {nodes [
   UpdateContextMenuButtonsVisibility() {is=title K2Node_FunctionEntry}
   > SetHideInputAction!(
    execute
    self:Common Button Base Object Reference=Get ContextMenuButton!(self:WBP UIKit Context Menu Anchor Button Base Object Reference=Get WBP_UIKit_ContextMenuAnchorButton()) -> (ContextMenuButton:WBP UIKit Button Generic Base Object Reference)
    bInHideInputAction:bool=true
   ) -> (then)
   n1 = GetRuntimeOptions()
   > SetVisibility(
    self=Get WBP_UIKit_ContextMenuAnchorButton()
    InVisibility=Select!(
     Option 0:ESlateVisibility Enum=Collapsed
     Option 1:ESlateVisibility Enum=SelfHitTestInvisible
     Index=Get bEnableContextMenuInDiscover(self=@n1)
    ) -> (ReturnValue:ESlateVisibility Enum)
   )
   n2 = Get ContextMenuActionContext()
   n3 = GetRuntimeOptions()
   > Set CanInputActionOnButton(CanInputActionOnButton:bool=AND Boolean(
     A=Get IsTileActive()
     B=Get bEnableContextMenuInDiscover(self=@n3)
     C=HasInteractionItems!(self:UIKit Context Menu Action Context Object Reference=@n2) -> (ReturnValue:bool)
    )) -> (Output_Get:bool)
   n4 = GetCurrentInputType(self=Get CommonInputSubsystem())
   > n5 = Branch(Condition=OR Boolean(A="Equal (Enum)"(A=@n4, B=Gamepad), B="Equal (Enum)"(A=@n4, B=Touch)))
   n6 = Get IsContextMenuIconAlwaysVisible() -> (IsContextMenuIconAlwaysVisible:bool)
   n7 = Get CanInputActionOnButton() -> (CanInputActionOnButton:bool)
   > SetVisibility(
    self=Get ContextMenuButton(self=Get WBP_UIKit_ContextMenuAnchorButton())
    InVisibility=Select(Option 0=Collapsed, Option 1=HitTestInvisible, Index=OR Boolean(A=@n7, B=@n6))
   )
   > SetTriggeringInputAction!(
    execute
    self:Common Button Base Object Reference=Get ContextMenuButton(self=Get WBP_UIKit_ContextMenuAnchorButton())
    InputActionRow:"Data Table Row Handle Structure (by ref)"=Get No Input Action Row() -> (No Input Action Row:Data Table Row Handle Structure)
   ) -> (then)
   > n8 = SetBottomButtonVisibility(self=Get WBP_BottomBarHackButton_ContextMenu(), ShouldButtonBeVisible=Get CanInputActionOnButton())
   Call On Context Menu Visibility Changed!(
    execute:Exec=[@n11.then, @n8.then]
    self:WBP Discover Tile Base Object Reference
    IsVisible:bool=OR Boolean(A=Get CanInputActionOnButton(), B=Get IsContextMenuIconAlwaysVisible())
   ) -> (then:Exec) {is=title K2Node_CallDelegate}
   n9 = Get CanInputActionOnButton()
   n10 = Get IsContextMenuIconAlwaysVisible()
   SetVisibility(
    execute=@n5.else
    self=Get ContextMenuButton(self=Get WBP_UIKit_ContextMenuAnchorButton())
    InVisibility=Select(Option 0=Hidden, Option 1=SelfHitTestInvisible, Index=OR Boolean(A=@n9, B=@n10))
   )
   > SetTriggeringInputAction(
    self=Get ContextMenuButton(self=Get WBP_UIKit_ContextMenuAnchorButton())
    InputActionRow=Select(
     Option 0:"Data Table Row Handle Structure (by ref)"=Get No Input Action Row()
     Option 1:"Data Table Row Handle Structure (by ref)"=Get ContextMenuInputAction() -> (ContextMenuInputAction:Data Table Row Handle Structure)
     Index=Get CanInputActionOnButton()
    ) -> (ReturnValue:"Data Table Row Handle Structure (by ref)")
   )
   > n11 = SetBottomButtonVisibility(self=Get WBP_BottomBarHackButton_ContextMenu(), ShouldButtonBeVisible=false)
  ]}
 UpdateIcon {nodes [
   UpdateIcon() {is=title K2Node_FunctionEntry}
   > Branch(Condition=IsValid(Object=Get Block_Icon()))
   n1 = Get Block_Icon()
   > SetMaterial!(
    execute
    self:WBP UIKit Block Image Object Reference=@n1
    InMaterial:"Material Instance Object Reference (by ref)"=Get IconMaterial() -> (IconMaterial:Material Instance Object Reference)
   ) -> (then)
   > SetVerticalAlignment(self=SlotAsGridSlot(Widget=@n1), InVerticalAlignment=VAlign_Center)
   > SetSize!(execute, self:WBP UIKit Block Image Object Reference=@n1, Desired Size:Vector 2D Structure=(X=24,Y=24)) -> (then)
   > SetPadding!(
    execute
    self:Grid Slot Object Reference=SlotAsGridSlot(Widget=@n1)
    InPadding:Margin Structure=Make Margin(Left=0, Top=0, Right=4, Bottom=0)
   ) -> (then)
  ]}
 UpdateIconVisibility {nodes [
   n1 = UpdateIconVisibility!() -> (then, IsVisible:bool=~) {is=title K2Node_FunctionEntry}
   n2 = Get Block_Icon()
   > Branch(Condition=IsValid(Object=@n2))
   > SetVisibility(
    self=@n2
    InVisibility=Select(Option 0:ESlateVisibility Enum=Collapsed, Option 1:ESlateVisibility Enum=HitTestInvisible, Index=@n1) -> (ReturnValue:ESlateVisibility Enum)
   )
  ]}
 UpdateNamedSlotVisibility {nodes [
   n1 = UpdateNamedSlotVisibility!() -> (then, IsVisible:bool=~) {is=title K2Node_FunctionEntry}
   n2 = Get Slot_Right()
   > Branch(Condition=IsValid(Object=@n2))
   > SetVisibility(self=@n2, InVisibility=Select(Option 0=Collapsed, Option 1=HitTestInvisible, Index=@n1))
  ]}
 UpdatePrimaryTextBlock {nodes [
   n1 = UpdatePrimaryTextBlock!() -> (then, UsePrimaryText:bool=~) {is=title K2Node_FunctionEntry}
   > n2 = Set UsePrimaryTextBlock(UsePrimaryTextBlock:bool=@n1) -> (Output_Get:bool)
   > n3 = Branch(Condition=@n2)
   > AddPrimaryTextBlock() {is=title K2Node_CallFunction}
   RemovePrimaryTextBlock(execute=@n3.else) {is=title K2Node_CallFunction}
  ]}
 UpdatePrimaryTextJustification {nodes [
   n1 = UpdatePrimaryTextJustification!() -> (then, Justification:ETextJustify Enum=Left) {is=title K2Node_FunctionEntry}
   > n2 = Set PrimaryTextJustification(PrimaryTextJustification:ETextJustify Enum=@n1) -> (Output_Get:ETextJustify Enum=Left)
   n3 = Get Block_PrimaryText()
   > Branch(Condition=IsValid(Object=@n3))
   > SetBlockTextJustification!(execute, self:WBP Discover Tile Text Primary Object Reference=@n3, TextJustification:ETextJustify Enum=@n2) -> (then)
  ]}
 UpdatePrimaryTextScrolling {nodes [
   n1 = UpdatePrimaryTextScrolling!() -> (then, IsMarqueeScrollingEnabled:bool=~) {is=title K2Node_FunctionEntry}
   n2 = Get Block_PrimaryText()
   > Branch(Condition=IsValid(Object=@n2))
   > SetBlockScrollingEnabled!(execute, self:WBP Discover Tile Text Primary Object Reference=@n2, IsMarqueeScrollingEnabled:bool=@n1) -> (then)
  ]}
 UpdateSecondaryTextBlock {nodes [
   n1 = UpdateSecondaryTextBlock!() -> (then, UseSecondaryText:bool=~) {is=title K2Node_FunctionEntry}
   > n2 = Set UseSecondaryTextBlock(UseSecondaryTextBlock:bool=@n1) -> (Output_Get:bool)
   > n3 = Branch(Condition=@n2)
   > AddSecondaryTextBlock() {is=title K2Node_CallFunction}
   RemoveSecondaryTextBlock(execute=@n3.else) {is=title K2Node_CallFunction}
  ]}
 UpdateSecondaryTextColorAndOpacity {nodes [
   n1 = UpdateSecondaryTextColorAndOpacity!() -> (then, UseDefault:bool=~, OverrideColor:Slate Color Structure) {is=title K2Node_FunctionEntry}
   > n2 = Set SecondaryTextColor(SecondaryTextColor:Slate Color Structure=@n1.OverrideColor) -> (Output_Get:Slate Color Structure)
   n3 = Get Block_SecondaryText()
   > Branch(Condition=IsValid(Object=@n3))
   > SetBlockTextColorAndOpacity!(execute, self:WBP Discover Tile Text Secondary Object Reference=@n3, UseDefault:bool=@n1, OverrideColor:Slate Color Structure=@n2) -> (then)
  ]}
 UpdateSecondaryTextJustification {nodes [
   n1 = UpdateSecondaryTextJustification!() -> (then, Justification:ETextJustify Enum=Left) {is=title K2Node_FunctionEntry}
   > n2 = Set SecondaryTextJustification(SecondaryTextJustification:ETextJustify Enum=@n1) -> (Output_Get:ETextJustify Enum=Left)
   n3 = Get Block_SecondaryText()
   > Branch(Condition=IsValid(Object=@n3))
   > SetBlockTextJustification!(execute, self:WBP Discover Tile Text Secondary Object Reference=@n3, Justification:ETextJustify Enum=@n2) -> (then)
  ]}
 UpdateSecondaryTextVisibility {nodes [
   n1 = UpdateSecondaryTextVisibility!() -> (then, IsVisible:bool=~) {is=title K2Node_FunctionEntry}
   n2 = Get Block_SecondaryText()
   > Branch(Condition=IsValid(Object=@n2))
   > SetVisibility(self=@n2, InVisibility=Select(Option 0=Collapsed, Option 1=HitTestInvisible, Index=@n1))
  ]}
]
variables [
 Block_Icon {type=WBP_UIKit_Block_Image_C, default=null}
 Block_PrimaryText {type=WBP_Discover_Tile_TextPrimary_C, default=null}
 Block_SecondaryText {type=WBP_Discover_Tile_TextSecondary_C, default=null}
 ContextMenuActionContext {type=UIKitContextMenuActionContext, default=null}
 ContextMenuInputAction {type=DataTableRowHandle, default {dataTable=/FortUILibrary/Foundation/ActivityBrowserInputActionTable, rowName=ContextMenu}}
 DebugGridSlot {type=GridSlot, default=null, class=~}
 DesignTimeIsSquareTiles {type=boolean}
 DynamicCalculatedTileSize {type=Vector2D, default=<304 171>}
 Icon {type=Texture2D}
 IconMaterial {type=MaterialInstance, default=/Game/UI/DiscoveryShared/Art/MI_UI_Icon_CCU}
 IsContextMenuIconAlwaysVisible {type=boolean}
 IsJoinablePartyTile {type=boolean}
 IsOutlineBlockCircular {type=boolean}
 IsTileActive {type=boolean}
 IsTileNameInside {type=boolean}
 No Input Action Row {type=DataTableRowHandle, default {dataTable=null, rowName=None}}
 OnContextMenuToggle {}
 OnContextMenuVisibilityChanged {}
 OnSetPrimaryTextFromVM {}
 OnSetTileContainerSizeOverride {}
 OnSetTileIndex {}
 ParentPanelName {type=string, default=0}
 PrimaryTextJustification {type=ETextJustify, default=Center}
 Right Clicked Sound {type=SoundBase, default=/Game/Sounds/Fort_UI_Sounds/FrontEnd/DefaultSounds/PressedButtons/MSS_UI_Button_Pressed_Default}
 RowHasViewAll {type=boolean}
 SecondaryTextClass {type=Class, default=&/WBP_Discover_Tile_TextSecondary.WBP_Discover_Tile_TextSecondary_C}
 SecondaryTextColor {type=SlateColor, default {colorUseRule=UseColor_Specified, specifiedColor=<rgba 0.9948 1 0.9896 1>}}
 SecondaryTextJustification {type=ETextJustify, default=Left}
 TextClass {type=Class, default=&/WBP_Discover_Tile_TextPrimary.WBP_Discover_Tile_TextPrimary_C}
 TileDebug {type=WBP_TileDebug_C, default=null}
 TileItemIndex {type=integer}
 TileItemViewModel {type=FortDiscoverTileItemVM, default=null}
 TileSize {type=Vector2D, default=<256 144>}
 UseIconBlock {type=boolean}
 UsePrimaryTextBlock {type=boolean}
 UseSecondaryTextBlock {type=boolean}
]
widgets {
 info {inheritedWidgetCount=0, parentClass=&CommonUserWidget, rootWidgetClass=&SizeBox, widgetCount=14}
 widgets [
  SizeBox_Main {widgetClassPath=&SizeBox, bIsVariable=true, widget=SizeBox_Main}
  Grid_Root_GrowsOnSelection {
   widgetClassPath=&GridPanel
   parent=SizeBox_Main
   widget=Grid_Root_GrowsOnSelection
   properties {columnFill [1], rowFill [1]}
   slot {class=SizeBoxSlot}
  }
  Grid_Main {
   widgetClassPath=&GridPanel
   bIsVariable=true
   parent=Grid_Root_GrowsOnSelection
   widget=Grid_Main
   properties {columnFill [1], rowFill [1]}
   slot {class=GridSlot}
  }
  Slot_TileContents {widgetClassPath=&NamedSlot, bIsVariable=true, parent=Grid_Main, widget=Slot_TileContents, slot {class=GridSlot}}
  WBP_CaptureForPostBufferUpdate {
   widgetClassPath=&WBP_CaptureForPostBufferUpdate
   bIsVariable=true
   parent=Grid_Main
   properties {visibility=Collapsed}
   slot {class=GridSlot, properties {layer=5}}
  }
  Block_Outline {
   widgetClassPath=/FortUILibrary/UIKit/Buttons/Blocks/WBP_UIKit_Block_Outline
   bIsVariable=true
   parent=Grid_Main
   widget=Block_Outline
   properties {bottomLeftCornerRadius=20, bottomRightCornerRadius=20, radiusInPixels=true, topLeftCornerRadius=20, topRightCornerRadius=20}
   slot {class=GridSlot, properties {layer=10}}
  }
  Slot_OverlayContent {
   widgetClassPath=&NamedSlot
   bIsVariable=true
   parent=Grid_Main
   widget=Slot_OverlayContent
   slot {class=GridSlot, properties {layer=20}}
  }
  WBP_UIKit_ContextMenuAnchorButton {
   widgetClassPath=/FortUILibrary/UIKit/ContextMenu/WBP_UIKit_ContextMenuAnchorButton
   bIsVariable=true
   parent=Grid_Root_GrowsOnSelection
   widget=WBP_UIKit_ContextMenuAnchorButton
   properties {
    buttonClass=/DiscoveryBrowser/Tiles/Islands/WBP_Discovery_ContextualButton.WBP_Discovery_ContextualButton_C
    buttonIcon=/FortUILibrary/Textures/Icons/T_UI_Icon_UIKit_ContextMenu_SDF
    buttonText=""
    menuWidgetClass=/DiscoveryBrowser/WBP_DiscoveryContextMenu_ModalDomain.WBP_DiscoveryContextMenu_ModalDomain_C
    useBottomBarButtons=true
    visibility=Collapsed
    button Padding=<ltrb 0 0 0 0>
   }
   slot {class=GridSlot, properties {horizontalAlignment=HAlign_Right, layer=20, verticalAlignment=VAlign_Bottom, padding=<ltrb 8 0 8 8>}}
  }
  WBP_BottomBarHackButton_ContextMenu {
   widgetClassPath=/Game/UI/Frontend/BottomBar/WBP_BottomBarDecoyButton
   bIsVariable=true
   parent=Grid_Root_GrowsOnSelection
   widget=WBP_BottomBarHackButton_ContextMenu
   properties {triggeringInputAction {dataTable=&ActivityBrowserInputActionTable, rowName=ContextMenu}}
   slot {class=GridSlot, properties {horizontalAlignment=HAlign_Right, verticalAlignment=VAlign_Bottom}}
  }
  VisualAttachment_OutsideContent {
   widgetClassPath=&FortVisualAttachment
   bIsVariable=true
   parent=Grid_Root_GrowsOnSelection
   widget=VisualAttachment_OutsideContent
   properties {bOverride_WidthOverride=true, visibility=HitTestInvisible}
   slot {class=GridSlot, properties {row=1}}
  }
  VerticalBox_0 {
   widgetClassPath=&VerticalBox
   parent=VisualAttachment_OutsideContent
   widget=VerticalBox_0
   slot {class=SizeBoxSlot, properties {padding=<ltrb 0 8 0 0>}}
  }
  PrimaryRow {
   widgetClassPath=&GridPanel
   bIsVariable=true
   parent=VerticalBox_0
   widget=PrimaryRow
   properties {columnFill [1,0], rowFill [1]}
   slot {class=VerticalBoxSlot}
  }
  SecondaryRow {
   widgetClassPath=&GridPanel
   bIsVariable=true
   parent=VerticalBox_0
   widget=SecondaryRow
   properties {columnFill [0,1,0], rowFill [1]}
   slot {class=VerticalBoxSlot}
  }
  Slot_Right {
   widgetClassPath=&NamedSlot
   bIsVariable=true
   parent=SecondaryRow
   widget=Slot_Right
   slot {class=GridSlot, properties {column=2, columnSpan=1}}
  }
 ]
}
