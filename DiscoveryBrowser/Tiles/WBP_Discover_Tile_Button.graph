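# WBP_Discover_Tile_Button is the holdable button for Discovery browser tiles, wrapping WBP_Discover_Tile_Base with named content and overlay slots.
# Hover, selection and focus loss drive the tile's active state, through a nested button group selection view model when valid, otherwise directly on the base.
# Only touch input makes it selectable; focus gain is broadcast.
# Holding toggles the context menu, also selecting on touch; mouse-down forwards to the base's right-click handler.
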WidgetBlueprint /DiscoveryBrowser/Tiles/WBP_Discover_Tile_Button
graphs [
 BP_OnHoldTriggered {nodes [
   n1 = On Hold Triggered!() -> (then, CurrentInputType:ECommonInputType Enum) {is=title K2Node_FunctionEntry}
   > n2 = Switch on ECommonInputType!(execute, Selection:ECommonInputType Enum=@n1) -> (MouseAndKeyboard:Exec, Gamepad:Exec, Touch:Exec, Count:Exec) {is=title K2Node_SwitchEnum}
   SetIsSelected!(execute=@n2.Touch, self:Common Button Base Object Reference, InSelected:bool=true, bGiveClickFeedback:bool=false) -> (then)
   n3 = Get WBP_Discover_Tile_Base() -> (WBP_Discover_Tile_Base:WBP Discover Tile Base Object Reference)
   > ToggleOpen!(
    execute
    self:Menu Anchor Object Reference=Get Anchor!(self:UIKit Context Menu Anchor Object Reference=Get WBP_UIKit_ContextMenuAnchorButton!(self:WBP Discover Tile Base Object Reference=@n3) -> (WBP_UIKit_ContextMenuAnchorButton:WBP UIKit Context Menu Anchor Button Object Reference)) -> (Anchor:Menu Anchor Object Reference)
    bFocusOnOpen:bool=false
   ) -> (then)
   > Return Node!(execute, ReturnValue:Event Reply Structure=Handled())
  ]}
 EventGraph {nodes [
   Event On Entry Released() {is=title K2Node_Event}
   > StopTileActiveImpressions(self:WBP Discover Tile Base Object Reference=Get WBP_Discover_Tile_Base())
   Event Construct()
   n1 = Create Event()
   n2 = Get CommonInputSubsystem()
   > Bind Event to On Input Method Changed(self=@n2, Delegate=@n1)
   > OnHandleInputMethodChanged!(execute, self, NewInputType:ECommonInputType Enum=GetCurrentInputType(self=@n2)) -> (then)
   Event On Hovered()
   n3 = NOT Boolean(A=Get bSelectable() -> (bSelectable:bool))
   > n4 = Branch(Condition=@n3)
   HandleTileSelectionActive!(execute=[@n8.then, @n4.then, @n6.then], self, InActive:bool=true) -> (then)
   Event On Focused() {is=title K2Node_Event}
   > Call On Focus Gained!(execute:Exec, self:WBP Discover Tile Button Object Reference, FocusedItem:WBP Discover Tile Button Object Reference=Self-Reference()) -> (then:Exec) {is=title K2Node_CallDelegate}
   n5 = IsUsingFocusNavigationOnKBM!() -> (IsTileUsingFocusForOutlineVisuals:bool)
   > n6 = Branch(Condition=@n5)
   Event Destruct()
   > Unbind Event from On Input Method Changed(self=@n2, Delegate=@n1)
   Event On Unhovered()
   > n7 = Branch(Condition=@n3)
   HandleTileSelectionActive(execute=[@n11.then, @n7.then, @n9.then], InActive=false)
   Event On Selected()
   > n8 = Branch(Condition=Get bSelectable())
   Event On Deselected()
   > n9 = Branch(Condition=Get bSelectable())
   Event On Unfocused() {is=title K2Node_Event}
   n10 = IsHovered()
   > Branch(Condition=AND Boolean(A=@n5, B=NOT Boolean(A=@n10)))
   > n11 = Branch(Condition=Get IsTileActive!(self:WBP Discover Tile Base Object Reference=Get WBP_Discover_Tile_Base()) -> (IsTileActive:bool))
   Event OnInitialized()
   > Override Right Click Sound!(
    execute
    self:WBP Discover Tile Base Object Reference=Get WBP_Discover_Tile_Base()
    New Right Click Sound:Sound Base Object Reference=Get Right Clicked Sound Override() -> (Right Clicked Sound Override:Sound Base Object Reference)
   ) -> (then)
  ]}
 GetNestedButtonGroupSelectionVM {nodes [
   GetNestedButtonGroupSelectionVM() {is=title K2Node_FunctionEntry}
   > Return Node(ReturnValue:Fort Nested Button Group Selection VM Object Reference)
  ]}
 HandleTileSelectionActive {nodes [
   HandleTileSelectionActive!() -> (then, InActive:bool=~) {is=title K2Node_FunctionEntry}
   > n1 = Is Valid(InputObject=GetNestedButtonGroupSelectionVM!(self:Fort Nested Button Group VMItem Interface) -> (ReturnValue:Fort Nested Button Group Selection VM Object Reference) {is=title K2Node_CallFunction})
   n2 = Branch(execute=@n1.Is Not Valid, Condition=Get InActive() -> (InActive:bool))
   n3 = Get WBP_Discover_Tile_Base()
   > OnTileBaseSelectionActive(self:WBP Discover Tile Base Object Reference=@n3)
   OnTileBaseSelectionNotActive(execute=@n2.else, self:WBP Discover Tile Base Object Reference=@n3)
   SetIsSelected!(
    execute=@n1.Is Valid
    self:Fort Nested Button Group Selection VM Object Reference=GetNestedButtonGroupSelectionVM()
    InButton:Fort Nested Button Group VMItem Interface=Self-Reference()
    bInSelected:bool=Get InActive()
   ) -> (then)
  ]}
 IsUsingFocusNavigationOnKBM {nodes [
   IsUsingFocusNavigationOnKBM() {is=title K2Node_FunctionEntry}
   n1 = GetCurrentInputType(self=Get CommonInputSubsystem())
   > Return Node!(
    execute
    IsTileUsingFocusForOutlineVisuals:bool=AND Boolean(A="Equal (Enum)"(A=@n1, B=MouseAndKeyboard), B=Get IsTileUsingFocusForOutlineVisuals() -> (IsTileUsingFocusForOutlineVisuals:bool))
   )
  ]}
 OnFocusGained {nodes [OnFocusGained!() -> (then, FocusedItem:WBP Discover Tile Button Object Reference) {is=title K2Node_FunctionEntry}]}
 OnHandleInputMethodChanged {nodes [
   n1 = OnHandleInputMethodChanged!() -> (then, NewInputType:ECommonInputType Enum) {is=title K2Node_FunctionEntry}
   > SetIsSelectable!(
    execute
    self:Common Button Base Object Reference
    bInIsSelectable:bool=Select!(MouseAndKeyboard:bool=false, Gamepad:bool=false, Touch:bool=true, Count:bool=false, Index:ECommonInputType Enum=@n1) -> (ReturnValue:bool)
   ) -> (then)
  ]}
 OnMouseButtonDown {nodes [
   n1 = OnMouseButtonDown!() -> (then, MyGeometry:Geometry Structure, MouseEvent:"Pointer Event Structure (by ref)") {is=title K2Node_FunctionEntry}
   > n2 = OnRightClickMouse!(execute, self:WBP Discover Tile Base Object Reference=Get WBP_Discover_Tile_Base(), MouseEvent:Pointer Event Structure=@n1.MouseEvent) -> (then, WasHandled:Event Reply Structure)
   > Return Node!(execute, ReturnValue:Event Reply Structure=@n2)
  ]}
 OnSelectedButtonChanged {nodes [
   n1 = OnSelectedButtonChanged!() -> (then, SelectedButton:Fort Nested Button Group VMItem Interface) {is=title K2Node_FunctionEntry}
   > n2 = Branch(Condition="Equal (Object)"(A:Fort Nested Button Group VMItem Interface=@n1, B=Self-Reference()))
   n3 = Get IsTileActive(self=Get WBP_Discover_Tile_Base())
   > n4 = Branch(Condition=@n3)
   n5 = Get WBP_Discover_Tile_Base()
   OnTileBaseSelectionActive(execute=@n4.else, self=@n5)
   Branch(execute=@n2.else, Condition=@n3)
   > OnTileBaseSelectionNotActive(self=@n5)
   > Branch(Condition=AND Boolean(A=GetSelected!(self:Common Button Base Object Reference) -> (ReturnValue:bool), B=Get bSelectable()))
   > SetSelectedInternal!(execute, self:Common Button Base Object Reference, bInSelected:bool=false, bAllowSound:bool=false, bBroadcast:bool=true) -> (then)
  ]}
]
variables [
 IsTileUsingFocusForOutlineVisuals {type=boolean}
 OnFocusGained {}
 Right Clicked Sound Override {type=SoundBase, default=null}
]
widgets {
 info {
  inheritedWidgetCount=1
  namedSlotCount=2
  parentClass=/Script/FortniteUI.UIKitHoldableModularButton
  rootWidgetClass=&GridPanel
  widgetCount=5
 }
 widgets [
  GridRoot {widgetClassPath=&GridPanel, widget=GridRoot, properties {columnFill [1], rowFill [1]}}
  WBP_Discover_Tile_Base {
   widgetClassPath=&/WBP_Discover_Tile_Base
   bIsVariable=true
   parent=GridRoot
   widget=WBP_Discover_Tile_Base
   slot {class=GridSlot, properties {horizontalAlignment=HAlign_Left, verticalAlignment=VAlign_Top}}
  }
  Slot_TileButtonContents {widgetClassPath=&NamedSlot, bIsVariable=true, namedSlotHost=WBP_Discover_Tile_Base, widget=Slot_TileButtonContents}
  Slot_TileOverlay {widgetClassPath=&NamedSlot, bIsVariable=true, namedSlotHost=WBP_Discover_Tile_Base, widget=Slot_TileOverlay}
  InputActionWidget {widgetClassPath=&CommonActionWidget, bInherited=true, widget=null}
 ]
}
